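# M_RoadSpline_01 is a Default Lit Landmass preview material for spline-mesh roads, shading asphalt with stripes.
# A SphereMask on the UV V coordinate blends between stripe and asphalt scalar values, placing asphalt in the central band and stripe toward the edges.
# That blend multiplies a power-adjusted macro-variation texture projected in world XY, so the grain is independent of mesh UVs.
# Scalar parameters expose mask shape, both brightness values, texture scale and power.
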Material /Landmass/PreviewContent/Roads/Mats/M_RoadSpline_01
graph {
 properties {bUsedWithSplineMeshes=true, pixelDepthOffsetMode=PDOM_Legacy, usedShadingModels=Default Lit}
 nodes [
  n1 = ComponentMask(None=TextureCoordinate()) {g=true}
  n2 = ScalarParameter() {defaultValue=0.5, parameterName=B}
  n3 = ScalarParameter() {defaultValue=0.55, parameterName=R}
  n4 = ScalarParameter() {defaultValue=0.8, parameterName=H}
  n5 = MaterialFunctionCall!(Scale=ScalarParameter() {defaultValue=1024, parameterName=Asphalt Scale}) {materialFunction=/Engine/Functions/Engine_MaterialFunctions01/Coordinates/WorldPosition-XY}
  Multiply(
   A=LinearInterpolate(
    A=ScalarParameter() {defaultValue=4, parameterName=Stripe D}
    B=ScalarParameter() {defaultValue=0.1, parameterName=Asphalt D}
    Alpha=SphereMask(A=@n1, B=@n2, Radius=@n3, Hardness=@n4)
   )
   B=Power(
    Base=TextureSample(UVs=@n5.Result) {texture=/Engine/EngineMaterials/T_Default_MacroVariation}.RGB
    Exp=ScalarParameter() {defaultValue=0.5, parameterName=Power}
   )
  )
 ]
}
